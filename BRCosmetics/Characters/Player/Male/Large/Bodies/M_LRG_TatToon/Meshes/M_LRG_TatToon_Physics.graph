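# Physics asset for the M_LRG_TatToon large male player body, set up for secondary motion on its chains and shirt back.
# Capsules on the upper spine and a necklace collider give torso volume around the dynamic chain and shirt-back bones.
# The two chain bones link in series from a small base body tied to spine_05, and the shirt flap hangs from its own base.
# Dynamic joints have tight, asymmetric swing limits and zero-degree twist, while base-to-spine and spine-to-spine joints rotate freely.
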PhysicsAsset /BRCosmetics/Characters/Player/Male/Large/Bodies/M_LRG_TatToon/Meshes/M_LRG_TatToon_Physics
bodies {
 base_chains_01_m [{shapeType=Capsule, length=0, radius=0.5, shapeName=base_chains_01_m_capsule, rotation=<rot 0 -90 0>}]
 base_necklaceCollider [{shapeType=Capsule, length=22.9334, radius=15, shapeName=spine_04_capsule, center=<11.0573 -1.2668 -10.7746>, rotation=<rot 0 0 90>}]
 base_shirtback_01_c [{shapeType=Capsule, length=0.505, radius=0.505, shapeName=base_shirtback_01_c_capsule, rotation=<rot 0 -90 0>}]
 dyn_chains_01_m [{shapeType=Capsule, length=10.5223, radius=5.2118, shapeName=dyn_chains_01_m_capsule, center=<7.52 0 1.4249>, rotation=<rot 74.3006 0 0>}]
 dyn_chains_02_m [{shapeType=Capsule, length=5.996, radius=5.536, shapeName=dyn_chains_02_m_capsule, center=<4.8099 0 3.8496>, rotation=<rot 0 0 90>}]
 dyn_shirtback_01_c [{
   shapeType=Capsule
   length=11.8532
   radius=4.1753
   shapeName=dyn_shirtback_01_c_capsule
   center=<8.0326 0 0>
   rotation=<rot 87.5226 -180 -180>
  }]
 spine_04 [{shapeType=Capsule, length=22.9334, radius=20.2261, shapeName=spine_04_capsule, center=<-1.3905 -1.8001 0>}]
 spine_05 [{shapeType=Capsule, length=15.9819, radius=9.8153, shapeName=spine_05_capsule}]
}
constraints [
 {bone1Name=dyn_chains_01_m, bone2Name=base_chains_01_m, swing1LimitDegrees=5, swing2LimitDegrees=30, twistLimitDegrees=0}
 {bone1Name=dyn_chains_02_m, bone2Name=dyn_chains_01_m, swing1LimitDegrees=5, swing2LimitDegrees=30, twistLimitDegrees=0}
 {
  bone1Name=base_chains_01_m
  bone2Name=spine_05
  swing1LimitDegrees=45
  swing1Motion=Free
  swing2LimitDegrees=45
  swing2Motion=Free
  twistLimitDegrees=45
  twistMotion=Free
 }
 {
  bone1Name=spine_05
  bone2Name=spine_04
  swing1LimitDegrees=45
  swing1Motion=Free
  swing2LimitDegrees=45
  swing2Motion=Free
  twistLimitDegrees=45
  twistMotion=Free
 }
 {bone1Name=dyn_shirtback_01_c, bone2Name=base_shirtback_01_c, swing1LimitDegrees=0, swing2LimitDegrees=30, twistLimitDegrees=0}
]
