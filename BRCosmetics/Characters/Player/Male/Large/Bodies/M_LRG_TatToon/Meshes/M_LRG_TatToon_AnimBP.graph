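# Post-process-style Anim Blueprint for the M_LRG_TatToon player body cosmetic, adding secondary motion to chains, belt, shirt back and belly.
# It copies the parent mesh pose through a Control Rig, then runs AnimDynamics and RigidBody simulation.
# Additive bone offsets and a belly-to-thigh constraint correct the cached result.
# The event graph tracks skydiving, parachute, weapon, front-end and mannequin states, though its deprecated gravity/RBAN function is not wired to update.
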AnimBlueprint /BRCosmetics/Characters/Player/Male/Large/Bodies/M_LRG_TatToon/Meshes/M_LRG_TatToon_AnimBP
target_skeleton=&/M_LRG_TatToon_Skeleton
graphs [
 AnimGraph {nodes [
   Copy Pose From Mesh() {bUseAttachedParent=true}
   n1 = "Transform (Modify) Bone - Bone: dyn_chains_01_m"!(
    ComponentPose=Local To Component(LocalPose=Use cached pose 'CRCache_Dynamics'())
    bAlphaBoolEnabled:bool=True
    Translation="0,0,1"
    Rotation="0,0,0"
    Scale="1,1,1"
   ) -> (Pose) {
    alphaInputType=Bool
    translationMode=BMM_Additive
    translationSpace=BCS_ParentBoneSpace
    alphaBoolBlend {blendOption=Linear}
    boneToModify {boneName=dyn_chains_01_m}
    translation {z=1}
   }
   n2 = "Transform (Modify) Bone - Bone: dyn_chains_02_m"!(ComponentPose=@n1, bAlphaBoolEnabled:bool=True, Translation="0,0,-1", Rotation="-5.651,0.002,-0.021", Scale="1,1,1") -> (Pose) {
    alphaInputType=Bool
    rotationMode=BMM_Additive
    rotationSpace=BCS_BoneSpace
    translationMode=BMM_Additive
    translationSpace=BCS_BoneSpace
    alphaBoolBlend {blendOption=Linear}
    boneToModify {boneName=dyn_chains_02_m}
    rotation=<rot -5.651 0.002 -0.021>
    translation {z=-1}
   }
   n3 = "Constraint - dyn_belly_r (P)"!(ComponentPose=@n2, bAlphaBoolEnabled:bool=True, ConstraintWeights_0=1) -> (Pose) {
    alphaInputType=Bool
    alphaBoolBlend {blendOption=Linear, blendOutTime=0.5}
    boneToModify {boneName=dyn_belly_r}
    constraintSetup [{targetBone {boneName=thigh_r}}]
    constraintWeights [1]
   }
   n4 = "Transform (Modify) Bone - Bone: dyn_belly_r"!(ComponentPose=@n3, bAlphaBoolEnabled:bool=True, Translation="7.080,-14.540,-11.696", Rotation="0,0,0", Scale="1,1,1") -> (Pose) {
    alphaInputType=Bool
    translationMode=BMM_Additive
    translationSpace=BCS_BoneSpace
    alphaBoolBlend {blendInTime=0.1, blendOption=Linear, blendOutTime=0.2}
    boneToModify {boneName=dyn_belly_r}
    translation=<7.08 -14.54 -11.696>
   }
   n5 = "Transform (Modify) Bone - Bone: base_shirtback_01_c"(ComponentPose=@n4, Alpha=1, Translation="0,0,0", Rotation="30,0,0", Scale="1,1,1") {
    rotationMode=BMM_Additive
    rotationSpace=BCS_BoneSpace
    translationMode=BMM_Additive
    translationSpace=BCS_BoneSpace
    alphaBoolBlend {blendOption=Linear}
    boneToModify {boneName=base_shirtback_01_c}
    rotation {pitch=30}
   }
   n6 = "Transform (Modify) Bone - Bone: base_chains_01_m"!(ComponentPose=@n5, bAlphaBoolEnabled:bool=True, Translation="0,0,0.5", Rotation="0,0,0", Scale="1,1,1") -> (Pose) {
    alphaInputType=Bool
    translationMode=BMM_Additive
    translationSpace=BCS_BoneSpace
    alphaBoolBlend {blendInTime=0.2, blendOption=Linear, blendOutTime=0.2}
    boneToModify {boneName=base_chains_01_m}
    translation {z=0.5}
   }
   n7 = "Transform (Modify) Bone - Bone: dyn_chains_02_m"(ComponentPose=@n6, bAlphaBoolEnabled=True, Translation="0,0,-0.5", Rotation="1,0,-3", Scale="1,1,1") {
    alphaInputType=Bool
    rotationMode=BMM_Additive
    rotationSpace=BCS_BoneSpace
    translationMode=BMM_Additive
    translationSpace=BCS_BoneSpace
    alphaBoolBlend {blendInTime=0.2, blendOption=Linear, blendOutTime=0.2}
    boneToModify {boneName=dyn_chains_02_m}
    rotation {pitch=1, roll=-3}
    translation {z=-0.5}
   }
   n8 = "Transform (Modify) Bone - Bone: base_chains_01_m"!(ComponentPose=@n7, Alpha:float=1, Translation="0,0,-0.9", Rotation="-0.5,0,-4", Scale="1,1,1") -> (Pose) {
    rotationMode=BMM_Additive
    rotationSpace=BCS_BoneSpace
    translationMode=BMM_Additive
    translationSpace=BCS_BoneSpace
    alphaBoolBlend {blendInTime=0.2, blendOption=Linear, blendOutTime=0.2}
    alphaScaleBiasClamp {bInterpResult=true}
    boneToModify {boneName=base_chains_01_m}
    rotation {pitch=-0.5, roll=-4}
    translation {z=-0.9}
   }
   n9 = "Transform (Modify) Bone - Bone: base_chains_01_m"!(ComponentPose=@n8, bAlphaBoolEnabled:bool=NOT Boolean(A=Get bRigidBodyEnabled()), Translation="0,0,0.8", Rotation="-3,0,2", Scale="1,1,1") -> (Pose) {
    alphaInputType=Bool
    rotationMode=BMM_Additive
    rotationSpace=BCS_BoneSpace
    translationMode=BMM_Additive
    translationSpace=BCS_BoneSpace
    alphaBoolBlend {blendInTime=0.2, blendOption=Linear, blendOutTime=0.2}
    alphaScaleBiasClamp {bInterpResult=true}
    boneToModify {boneName=base_chains_01_m}
    rotation {pitch=-3, roll=2}
    translation {z=0.8}
   }
   n10 = "Transform (Modify) Bone - Bone: dyn_chains_02_m"(ComponentPose=@n9, bAlphaBoolEnabled=True, Translation="0,0,0", Rotation="-5,0,0", Scale="1,1,1") {
    alphaInputType=Bool
    rotationMode=BMM_Additive
    rotationSpace=BCS_BoneSpace
    translationMode=BMM_Additive
    translationSpace=BCS_BoneSpace
    alphaBoolBlend {blendInTime=0.2, blendOption=Linear, blendOutTime=0.2}
    boneToModify {boneName=dyn_chains_02_m}
    rotation {pitch=-5}
   }
   n11 = "Transform (Modify) Bone - Bone: dyn_chains_01_m"(ComponentPose=@n10, bAlphaBoolEnabled=True, Translation="0,0,0", Rotation="5,0,5", Scale="1,1,1") {
    alphaInputType=Bool
    rotationMode=BMM_Additive
    rotationSpace=BCS_BoneSpace
    translationMode=BMM_Additive
    translationSpace=BCS_BoneSpace
    alphaBoolBlend {blendInTime=0.2, blendOption=Linear, blendOutTime=0.2}
    boneToModify {boneName=dyn_chains_01_m}
    rotation {pitch=5, roll=5}
   }
   Output Pose(Result=Component To Local(ComponentPose=@n11)) {layerGroup=None}
   n12 = Control Rig!(
    Source=Copy Pose From Mesh() {bUseAttachedParent=true}
    Alpha
    IsEmoting:bool=Get bIsPlayingEmote()
    IsMannequin:bool="Get isMannequin "() -> ("isMannequin ":bool)
   ) -> (Pose) {
    alphaBoolBlend {blendOption=Linear}
    controlRigAssetReference {blueprintRigClass=&/M_LRG_TatToon_CtrlRig.M_LRG_TatToon_CtrlRig_C}
   }
   Save cached pose 'CRCache_Input'(Pose=@n12) {cacheName=CRCache_Input}
   n13 = Anim Dynamics - Bone: dyn_shirtback_01_c!(
    SimulationSpace:AnimPhysSimSpaceType Enum=World
    ComponentPose=Local To Component(LocalPose=Use cached pose 'CRCache_Input'())
    bAlphaBoolEnabled:bool=NOT Boolean(A=Get bIsInFrontEnd())
    GravityOverride:Vector="0,0,-980"
    ExternalForce="0,0,0"
   ) -> (Pose) {
    alphaInputType=Bool
    angularDampingOverride=1
    angularSpringConstant=15000
    bAngularSpring=true
    bOverrideAngularDamping=true
    bUseGravityOverride=true
    gravityScale=0.35
    simulationSpace=World
    alphaBoolBlend {blendOption=Linear}
    boundBone {boneName=dyn_shirtback_01_c}
    componentAppliedLinearAccClamp=<5000 5000 5000>
    componentLinearAccScale {y=1}
    componentLinearVelScale {y=1}
    gravityOverride {z=-980}
    physicsBodyDefinitions [{
      boundBone {boneName=dyn_shirtback_01_c}
      boxExtents=<35 35 35>
      constraintSetup {angularLimitsMax=<0 0 30>, angularLimitsMin=<0 -70 -30>, angularTarget=<1 0 0>}
      localJointOffset=<2 0 0>
     }]
   }
   n14 = "Transform (Modify) Bone - Bone: base_belt_01_l"!(ComponentPose=@n13, bAlphaBoolEnabled:bool=True, Translation="0,0,0", Rotation="-0.159,-21.277,0.046", Scale="1,1,1") -> (Pose) {
    alphaInputType=Bool
    rotationMode=BMM_Additive
    rotationSpace=BCS_BoneSpace
    translationMode=BMM_Additive
    translationSpace=BCS_BoneSpace
    alphaBoolBlend {blendInTime=0.2, blendOption=Linear, blendOutTime=0.2}
    boneToModify {boneName=base_belt_01_l}
    rotation=<rot -0.159 -21.277 0.046>
   }
   n15 = Anim Dynamics - Bone: dyn_belt_01_l!(ComponentPose=@n14, Alpha=1, GravityOverride:Vector="0,0,0", ExternalForce="0,0,0") -> (Pose) {
    angularDampingOverride=1
    angularSpringConstant=20000
    bAngularSpring=true
    bOverrideAngularDamping=true
    bUseGravityOverride=true
    gravityScale=0.35
    simulationSpace=RootRelative
    alphaBoolBlend {blendOption=Linear}
    boundBone {boneName=dyn_belt_01_l}
    componentAppliedLinearAccClamp=<5000 5000 5000>
    componentLinearAccScale=<0.5 0.5 0.5>
    componentLinearVelScale=<0.5 0.5 1>
    gravityOverride {z=-980}
    physicsBodyDefinitions [{
      boundBone {boneName=dyn_belt_01_l}
      boxExtents=<30 30 30>
      constraintSetup {angularLimitsMax=<0 15 10>, angularLimitsMin=<0 -60 -30>, angularTarget=<1 0 0>}
      localJointOffset=<0.5 0 0>
     }]
    relativeSpaceBone {boneName=pelvis}
   }
   n16 = RigidBody!(ComponentPose=@n15, Alpha, OverrideWorldGravity:Vector="0,0,-980", ExternalForce) -> (Pose) {
    bOverrideWorldGravity=true
    overridePhysicsAsset=&/M_LRG_TatToon_Physics
    simulationSpace=BaseBoneSpace
    alphaBoolBlend {blendOption=Linear}
    baseBoneRef {boneName=spine_04}
    simSpaceSettings {
     maxAngularAcceleration=1
     maxAngularVelocity=3
     maxLinearAcceleration=2000
     maxLinearVelocity=500
     velocityScaleZ=0.5
     worldAlpha=1
     externalLinearDragV=<0.2 0.2>
    }
   }
   Save cached pose 'CRCache_Dynamics'(Pose=Component To Local(ComponentPose=@n16)) {cacheName=CRCache_Dynamics}
  ]}
 EventGraph {nodes [
   n1 = Gravity Override State Function DEPRECATED(
    CharacterPartAnimInstance=Self-Reference()
    GameTimeInSeconds=GetGameTimeInSeconds()
    GravityOverrideParams=Get Gravity Override Params()
    RBAN Params=Get RBAN Params()
   )
   > Set GravityOverride(GravityOverride=@n1)
   > Set RBAN Settings Out(RBAN Settings Out:RBANSetting Struct Structure=@n1.RBAN SettingsOut) -> (Output_Get:RBANSetting Struct Structure)
   Event BlueprintUpdateAnimation()
   > n2 = Sequence!(execute) -> (then_0, then_1, then_2:Exec, then_3:Exec)
   n3 = Branch(execute=@n2.then_1, Condition=Get bIsSkyDiving())
   > Set SimlationSpace(SimlationSpace:AnimPhysSimSpaceType Enum=Component) -> (Output_Get:AnimPhysSimSpaceType Enum)
   Set SimlationSpace(execute=@n3.else, SimlationSpace=World)
   n4 = AND Boolean(A=Get bIsInFrontEnd(), B=NOT Boolean(A=Get bIsPlayingEmote()))
   n5 = Set idleFrontEnd(execute=@n2.then_2, idleFrontEnd:bool=@n4) -> (Output_Get:bool)
   > n6 = Branch(Condition=@n5)
   > n7 = Set Skel Mesh(Skel Mesh=GetPartSkeletalMeshComponent())
   > Set FingerRot(FingerRot:double=GetAlphaFromReferenceTransformDelta(SkeletalMeshComponent=@n7, BoneName=index_metacarpal_r, BaseName=index_01_r, InputRange=(X=20,Y=40), TwistAxis="0,0,1")) -> (Output_Get:double)
   Set FingerRot(execute=@n6.else, FingerRot=0)
   n8 = GetOwningActor()
   Cast To FortPlayerMannequin!(execute=@n2.then_3, Object=@n8) -> (then, CastFailed, AsFort Player Mannequin:Fort Player Mannequin Object Reference) {is=title K2Node_DynamicCast}
   > n9 = For Each Loop(Array:Array of Skeletal Mesh Component Object References=Get Components By Class!(self:Actor Object Reference=@n8, ComponentClass:Actor Component Class Reference=/Script/Engine.SkeletalMeshComponent) -> (ReturnValue:Array of Skeletal Mesh Component Object References)) -> (Array Element:Skeletal Mesh Component Object Reference)
   > n10 = Branch(Condition="Equal (Object)"(
     A=Get SkeletalMesh!(self:Skinned Mesh Component Object Reference=@n9) -> (SkeletalMesh:Skeletal Mesh Object Reference)
     B=&/M_LRG_TatToon
    ))
   > "Set isMannequin "("isMannequin ":bool=false) -> (Output_Get:bool)
   "Set isMannequin "(execute=@n10.else, "isMannequin "=true)
   n11 = Cast To FortPlayerAnimInstance(execute=@n2.then_0, Object=GetAnimInstance(self=GetOwnerSkeletalMeshComponent()))
   > Set goingCommando(goingCommando:bool=false) -> (Output_Get:bool)
   > Set ArmMeele(ArmMeele:bool=false) -> (Output_Get:bool)
   > n12 = Switch on EFortWeaponCoreAnimation!(execute, Selection:EFortWeaponCoreAnimation Enum=Get WeaponCoreAnim(self=@n11)) -> (
    Melee:Exec
    Pistol:Exec
    Shotgun:Exec
    PaperBlueprint:Exec
    Rifle:Exec
    MeleeOneHand:Exec
    MachinePistol:Exec
    RocketLauncher:Exec
    GrenadeLauncher:Exec
    GoingCommando:Exec
    AssaultRifle:Exec
    TacticalShotgun:Exec
    SniperRifle:Exec
    TrapPlacement:Exec
    ShoulderLauncher:Exec
    AbilityDecoTool:Exec
    Crossbow:Exec
    C4:Exec
    RemoteControl:Exec
    DualWield:Exec
    AR_BullPup:Exec
    AR_ForwardGrip:Exec
    MedPackPaddles:Exec
    SMG_P90:Exec
    AR_DrumGun:Exec
    Consumable_Small:Exec
    Consumable_Large:Exec
    Balloon:Exec
    MountedTurret:Exec
    CreativeTool:Exec
    ExplosiveBow:Exec
    AshtonIndigo:Exec
    AshtonChicago:Exec
    MeleeDualWield:Exec
    Unarmed:Exec
   )
   Branch(execute=@n12.GoingCommando)
   > Set goingCommando(goingCommando=1)
   > Branch(Condition=Get bIsSkyDiving())
   > Set goingCommando(goingCommando=false)
   Branch(execute=@n12.Melee, Condition=@n4)
   > Set ArmMeele(ArmMeele=true)
   Event OnParachuteOpened()
   > Set openParachute(openParachute:double=1) -> (Output_Get:double)
   > Delay(Duration=2.5)
   > Set openParachute(openParachute=0)
  ]}
]
variables [
 ArmMeele {type=boolean}
 FingerRot {type=number}
 goingCommando {type=boolean}
 Gravity Override Params {type=GravityOverrideParamsStruct}
 GravityOverride {type=Vector}
 idleFrontEnd {type=boolean}
 "isMannequin " {type=boolean}
 openParachute {type=number}
 RBAN Params {
  type=RBANParamsStruct
  default {
   dBNO {simSpaceSettings {
     maxAngularAcceleration=1000
     maxAngularVelocity=3
     maxLinearAcceleration=300
     maxLinearVelocity=700
     velocityScaleZ=0
     externalLinearDragV=<0 0.5 0>
    }}
   dBNO_Carried {simSpaceSettings {
     maxAngularAcceleration=1000
     maxAngularVelocity=3
     maxLinearAcceleration=300
     maxLinearVelocity=700
     velocityScaleZ=0
     externalLinearDragV=<0 0.5 0>
    }}
   emoteOrMelee {simSpaceSettings {
     maxAngularAcceleration=1000
     maxAngularVelocity=3
     maxLinearAcceleration=300
     maxLinearVelocity=700
     velocityScaleZ=0
     externalLinearDragV=<0 0.5 0>
    }}
   front_End {simSpaceSettings {
     maxAngularAcceleration=1000
     maxAngularVelocity=3
     maxLinearAcceleration=300
     maxLinearVelocity=700
     velocityScaleZ=0
     externalLinearDragV=<0 0.5 0>
    }}
   onGround_Crouch_Moving {simSpaceSettings {
     maxAngularAcceleration=1000
     maxAngularVelocity=3
     maxLinearAcceleration=300
     maxLinearVelocity=700
     velocityScaleZ=0
     externalLinearDragV=<0 0.5 0>
    }}
   onGround_Falling {simSpaceSettings {
     maxAngularAcceleration=1000
     maxAngularVelocity=3
     maxLinearAcceleration=300
     maxLinearVelocity=700
     velocityScaleZ=0
     externalLinearDragV=<0 0.5 0>
    }}
   onGround_Moving {simSpaceSettings {
     maxAngularAcceleration=1000
     maxAngularVelocity=3
     maxLinearAcceleration=300
     maxLinearVelocity=700
     velocityScaleZ=0
     externalLinearDragV=<0 0.5 0>
    }}
   onGround_Standing {simSpaceSettings {
     maxAngularAcceleration=1000
     maxAngularVelocity=3
     maxLinearAcceleration=300
     maxLinearVelocity=700
     velocityScaleZ=0
     externalLinearDragV=<0 0.5 0>
    }}
   rail_Grinding {simSpaceSettings {
     maxAngularAcceleration=1000
     maxAngularVelocity=3
     maxLinearAcceleration=300
     maxLinearVelocity=700
     velocityScaleZ=0
     worldAlpha=0.1
     externalLinearDragV=<0 0.5 0>
    }}
   riding_Moving {simSpaceSettings {
     maxAngularAcceleration=1000
     maxAngularVelocity=3
     maxLinearAcceleration=300
     maxLinearVelocity=700
     velocityScaleZ=0
     externalLinearDragV=<0 0.5 0>
    }}
   riding_Moving_Targeting {simSpaceSettings {
     maxAngularAcceleration=1000
     maxAngularVelocity=3
     maxLinearAcceleration=300
     maxLinearVelocity=700
     velocityScaleZ=0
     externalLinearDragV=<0 0.5 0>
    }}
   riding_Standing {simSpaceSettings {
     maxAngularAcceleration=1000
     maxAngularVelocity=3
     maxLinearAcceleration=300
     maxLinearVelocity=700
     velocityScaleZ=0
     externalLinearDragV=<0 0.5 0>
    }}
   skydive_DivingDown {simSpaceSettings {
     maxAngularAcceleration=1000
     maxAngularVelocity=3
     maxLinearAcceleration=300
     maxLinearVelocity=700
     velocityScaleZ=0
     externalLinearDragV=<0 0.5 0>
    }}
   skydive_FloatingDown {
    componentLinearAccScale=<0 0.5 0>
    componentLinearVelScale=<0 0.5 0>
    simSpaceSettings {
     maxAngularAcceleration=1000
     maxAngularVelocity=3
     maxLinearAcceleration=300
     maxLinearVelocity=700
     velocityScaleZ=0
     externalLinearDragV=<0 0.5 0>
    }
   }
   skydive_GoingUp {simSpaceSettings {
     maxAngularAcceleration=1000
     maxAngularVelocity=3
     maxLinearAcceleration=300
     maxLinearVelocity=700
     velocityScaleZ=0
     externalLinearDragV=<0 0.5 0>
    }}
   skydive_Parachuting {
    componentLinearAccScale=<0 0.1 0.3>
    componentLinearVelScale=<0 0.1 0.3>
    simSpaceSettings {
     maxAngularAcceleration=1000
     maxAngularVelocity=3
     maxLinearAcceleration=300
     maxLinearVelocity=700
     velocityScaleZ=0
     externalLinearDragV=<0 0.5 0>
    }
   }
   skydive_Surfing {simSpaceSettings {
     maxAngularAcceleration=1000
     maxAngularVelocity=3
     maxLinearAcceleration=300
     maxLinearVelocity=700
     velocityScaleZ=0
     externalLinearDragV=<0 0.5 0>
    }}
   swimming {simSpaceSettings {
     maxAngularAcceleration=1000
     maxAngularVelocity=3
     maxLinearAcceleration=300
     maxLinearVelocity=700
     velocityScaleZ=0
     externalLinearDragV=<0 0.5 0>
    }}
   tactical_Sprint {simSpaceSettings {
     maxAngularAcceleration=1000
     maxAngularVelocity=3
     maxLinearAcceleration=300
     maxLinearVelocity=700
     velocityScaleZ=0
     externalLinearDragV=<0 0.5 0>
    }}
   ziplining {simSpaceSettings {
     maxAngularAcceleration=1000
     maxAngularVelocity=3
     maxLinearAcceleration=300
     maxLinearVelocity=700
     velocityScaleZ=0
     externalLinearDragV=<0 0.5 0>
    }}
  }
 }
 RBAN Settings Out {
  type=RBANSettingStruct
  default {
   componentLinearAccScale=<0 0 0>
   componentLinearVelScale=<0 0 0>
   simSpaceSettings {
    dampingAlpha=1
    maxAngularAcceleration=1000
    maxAngularVelocity=3
    maxLinearAcceleration=300
    maxLinearVelocity=700
    velocityScaleZ=0
    worldAlpha=1
    externalAngularVelocity=<0 0 0>
    externalLinearDragV=<0 0.5 0>
    externalLinearVelocity=<0 0 0>
   }
  }
 }
 run {type=boolean}
 SimlationSpace {type=AnimPhysSimSpaceType}
 Skel Mesh {type=SkeletalMeshComponent}
]
