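# Animation Blueprint for the M_LRG_TatToon face accessory, posing it from the character and adding secondary motion to a bow.
# Copies the attached parent's pose, layering Body_Mesh's pose over clavicles and calves when not sprinting, then applies additive pec and upper-arm tweaks.
# The cached result drives bow offsets around two spring-based Anim Dynamics bones.
# Each update stores body deform-bone transforms, the sprint weight and head-versus-neck bow anti-clipping alphas.
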AnimBlueprint /BRCosmetics/Characters/Player/Male/Large/Bodies/M_LRG_TatToon/Meshes/Parts/M_LRG_TatToon_FaceAcc_AnimBP
target_skeleton=&/M_LRG_TatToon_FaceAcc_Skeleton
graphs [
 AnimGraph {nodes [
   "Transform (Modify) Bone - Bone: upperarm_twist_01_r"(Alpha=1, Translation="0,1,0", Rotation="0,0,0", Scale="1,1,1") {
    rotationMode=BMM_Additive
    rotationSpace=BCS_BoneSpace
    translationMode=BMM_Additive
    translationSpace=BCS_BoneSpace
    alphaBoolBlend {blendOption=Linear}
    boneToModify {boneName=upperarm_twist_01_r}
    translation {y=1}
   }
   n1 = "Transform (Modify) Bone - Bone: base_bow_m"(
    ComponentPose=Local To Component(LocalPose=Use cached pose 'CRCache_jointPosition'())
    Alpha=1
    Translation="0,0,0"
    Rotation="36.598,47.172,37.190"
    Scale="1,1,1"
   ) {
    rotationMode=BMM_Additive
    rotationSpace=BCS_BoneSpace
    translationMode=BMM_Additive
    translationSpace=BCS_BoneSpace
    alphaBoolBlend {blendOption=Linear}
    boneToModify {boneName=base_bow_m}
    rotation=<rot 36.598 47.172 37.19>
   }
   n2 = Anim Dynamics - Bone: dyn_bow_01(ComponentPose=@n1, Alpha=1, ExternalForce="0,0,0") {
    angularDampingOverride=1
    angularSpringConstant=20000
    bAngularSpring=true
    bOverrideAngularDamping=true
    bUseGravityOverride=true
    gravityScale=0.35
    alphaBoolBlend {blendOption=Linear}
    boundBone {boneName=dyn_bow_01}
    componentLinearAccScale=<1 1 1>
    componentLinearVelScale=<1 1 1>
    gravityOverride {z=-980}
    physicsBodyDefinitions [{
      boundBone {boneName=dyn_bow_01}
      boxExtents=<20 30 30>
      constraintSetup {angularLimitsMax=<15 15 0>, angularLimitsMin=<-15 -15 0>, angularTarget=<1 0 0>}
      localJointOffset=<4 0 0>
     }]
   }
   n3 = Anim Dynamics - Bone: dyn_bow_02(ComponentPose=@n2, Alpha=1, ExternalForce="0,0,0") {
    angularDampingOverride=1
    angularSpringConstant=20000
    bAngularSpring=true
    bOverrideAngularDamping=true
    bUseGravityOverride=true
    gravityScale=0.35
    alphaBoolBlend {blendOption=Linear}
    boundBone {boneName=dyn_bow_02}
    componentLinearAccScale=<1 1 1>
    componentLinearVelScale=<1 1 1>
    gravityOverride {z=-980}
    physicsBodyDefinitions [{
      boundBone {boneName=dyn_bow_02}
      boxExtents=<20 30 30>
      constraintSetup {angularLimitsMax=<15 15 0>, angularLimitsMin=<-15 -15 0>, angularTarget=<1 0 0>}
      localJointOffset=<4 0 0>
     }]
   }
   n4 = "Transform (Modify) Bone - Bone: dyn_bow_m"(ComponentPose=@n3, Alpha=1, Translation="-0.5,0.5,2", Rotation="38.905,0.000,-0.000", Scale="1,1,1") {
    rotationMode=BMM_Additive
    rotationSpace=BCS_BoneSpace
    translationMode=BMM_Additive
    translationSpace=BCS_BoneSpace
    alphaBoolBlend {blendInTime=0.2, blendOption=Linear, blendOutTime=0.2}
    boneToModify {boneName=dyn_bow_m}
    rotation=<rot 38.905 0 0>
    translation=<-0.5 0.5 2>
   }
   Output Pose(Result=Component To Local(ComponentPose=@n4)) {layerGroup=None}
   n5 = Layered blend per bone(
    BasePose=Copy Pose From Mesh() {bUseAttachedParent=true}
    BlendPoses_0=Copy Pose From Mesh(SourceMeshComponent=Get Body_Mesh())
    BlendWeights_0=Get sprintFloat() -> (sprintFloat:double)
   ) {layerSetup [{branchFilters [{boneName=clavicle_l}, {boneName=clavicle_r}, {boneName=calf_l}, {boneName=calf_r}]}]}
   n6 = "Transform (Modify) Bone - Bone: deform_pec_l"(ComponentPose=Local To Component(LocalPose=@n5), Alpha=1, Translation="0,0,0", Rotation="0,0,0", Scale="1,1,1") {
    rotationMode=BMM_Additive
    rotationSpace=BCS_BoneSpace
    translationMode=BMM_Additive
    translationSpace=BCS_BoneSpace
    alphaBoolBlend {blendOption=Linear}
    boneToModify {boneName=deform_pec_l}
   }
   n7 = "Transform (Modify) Bone - Bone: deform_pec_r"(ComponentPose=@n6, Alpha=1, Translation="0,0,0", Rotation="0,0,0", Scale="1,1,1") {
    rotationMode=BMM_Additive
    rotationSpace=BCS_BoneSpace
    translationMode=BMM_Additive
    translationSpace=BCS_BoneSpace
    alphaBoolBlend {blendOption=Linear}
    boneToModify {boneName=deform_pec_r}
   }
   n8 = "Transform (Modify) Bone - Bone: deform_pec_aim_l"(ComponentPose=@n7, Alpha=1, Translation="0,0,0", Rotation="0,0,0", Scale="1,1,1") {
    rotationMode=BMM_Additive
    rotationSpace=BCS_BoneSpace
    translationMode=BMM_Additive
    translationSpace=BCS_BoneSpace
    alphaBoolBlend {blendOption=Linear}
    boneToModify {boneName=deform_pec_aim_l}
   }
   n9 = "Transform (Modify) Bone - Bone: deform_pec_aim_r"(ComponentPose=@n8, Alpha=1, Translation="0,0,0", Rotation="0,0,0", Scale="1,1,1") {
    rotationMode=BMM_Additive
    rotationSpace=BCS_BoneSpace
    translationMode=BMM_Additive
    translationSpace=BCS_BoneSpace
    alphaBoolBlend {blendOption=Linear}
    boneToModify {boneName=deform_pec_aim_r}
   }
   n10 = "Transform (Modify) Bone - Bone: deform_upperarm_fr_l"(ComponentPose=@n9, Alpha=1, Translation="0,1,0", Rotation="0,0,0", Scale="1,1,1") {
    rotationMode=BMM_Additive
    rotationSpace=BCS_BoneSpace
    translationMode=BMM_Additive
    translationSpace=BCS_BoneSpace
    alphaBoolBlend {blendOption=Linear}
    boneToModify {boneName=deform_upperarm_fr_l}
    translation {y=1}
   }
   n11 = "Transform (Modify) Bone - Bone: deform_upperarm_fr_r"(ComponentPose=@n10, Alpha=1, Translation="0,1,0", Rotation="0,0,0", Scale="1,1,1") {
    rotationMode=BMM_Additive
    rotationSpace=BCS_BoneSpace
    translationMode=BMM_Additive
    translationSpace=BCS_BoneSpace
    alphaBoolBlend {blendOption=Linear}
    boneToModify {boneName=deform_upperarm_fr_r}
    translation {y=1}
   }
   n12 = "Transform (Modify) Bone - Bone: deform_upperarm_up_l"(ComponentPose=@n11, Alpha=1, Translation="0,1,0", Rotation="0,0,0", Scale="1,1,1") {
    rotationMode=BMM_Additive
    rotationSpace=BCS_BoneSpace
    translationMode=BMM_Additive
    translationSpace=BCS_BoneSpace
    alphaBoolBlend {blendOption=Linear}
    boneToModify {boneName=deform_upperarm_up_l}
    translation {y=1}
   }
   n13 = "Transform (Modify) Bone - Bone: deform_upperarm_up_r"(ComponentPose=@n12, Alpha=1, Translation="0,1,0", Rotation="0,0,0", Scale="1,1,1") {
    rotationMode=BMM_Additive
    rotationSpace=BCS_BoneSpace
    translationMode=BMM_Additive
    translationSpace=BCS_BoneSpace
    alphaBoolBlend {blendOption=Linear}
    boneToModify {boneName=deform_upperarm_up_r}
    translation {y=1}
   }
   Save cached pose 'CRCache_jointPosition'(Pose=Component To Local(ComponentPose=@n13)) {cacheName=CRCache_jointPosition}
   Get Skel_Mesh()
  ]}
 EventGraph {nodes [
   Event BlueprintUpdateAnimation()
   > n1 = Sequence()
   n2 = BreakTransform(InTransform=GetSocketTransform(self=Get Body_Mesh(), InSocketName=deform_pec_l, TransformSpace=RTS_ParentBoneSpace))
   > Set deform_pec_l_Translation(deform_pec_l_Translation:Vector=@n2) -> (Output_Get:Vector)
   > Set deform_pec_l_Rotation(deform_pec_l_Rotation:Rotator=@n2.Rotation) -> (Output_Get:Rotator)
   n3 = BreakTransform(InTransform=GetSocketTransform(self=Get Body_Mesh(), InSocketName=deform_pec_r, TransformSpace=RTS_ParentBoneSpace))
   > Set deform_pec_r_Translation(deform_pec_r_Translation:Vector=@n3) -> (Output_Get:Vector)
   > Set deform_pec_r_Rotation(deform_pec_r_Rotation:Rotator=@n3.Rotation) -> (Output_Get:Rotator)
   n4 = BreakTransform(InTransform=GetSocketTransform(self=Get Body_Mesh(), InSocketName=deform_pec_aim_l, TransformSpace=RTS_ParentBoneSpace))
   > Set deform_pec_base_l_Translation(deform_pec_base_l_Translation:Vector=@n4) -> (Output_Get:Vector)
   > Set deform_pec_base_l_Rotation(deform_pec_base_l_Rotation:Rotator=@n4.Rotation) -> (Output_Get:Rotator)
   n5 = BreakTransform(InTransform=GetSocketTransform(self=Get Body_Mesh(), InSocketName=deform_pec_aim_r, TransformSpace=RTS_ParentBoneSpace))
   > Set deform_pec_base_r_Translation(deform_pec_base_r_Translation:Vector=@n5) -> (Output_Get:Vector)
   > Set deform_pec_base_r_Rotation(deform_pec_base_r_Rotation:Rotator=@n5.Rotation) -> (Output_Get:Rotator)
   n6 = BreakTransform(InTransform=GetSocketTransform(self=Get Body_Mesh(), InSocketName=deform_upperarm_up_l, TransformSpace=RTS_ParentBoneSpace))
   > Set deform_upper_up_l_Translation(deform_upper_up_l_Translation:Vector=@n6) -> (Output_Get:Vector)
   > Set deform_upper_up_l_Rotation(deform_upper_up_l_Rotation:Rotator=@n6.Rotation) -> (Output_Get:Rotator)
   n7 = BreakTransform(InTransform=GetSocketTransform(self=Get Body_Mesh(), InSocketName=deform_upperarm_up_r, TransformSpace=RTS_ParentBoneSpace))
   > Set deform_upper_up_r_Translation(deform_upper_up_r_Translation:Vector=@n7) -> (Output_Get:Vector)
   > Set deform_upper_up_r_Rotation(deform_upper_up_r_Rotation:Rotator=@n7.Rotation) -> (Output_Get:Rotator)
   n8 = BreakTransform(InTransform=GetSocketTransform(self=Get Body_Mesh(), InSocketName=deform_upperarm_fr_l, TransformSpace=RTS_ParentBoneSpace))
   > Set deform_upperarm_fr_l_Translation(deform_upperarm_fr_l_Translation:Vector=@n8) -> (Output_Get:Vector)
   > Set deform_upperarm_fr_l_Rotation(deform_upperarm_fr_l_Rotation:Rotator=@n8.Rotation) -> (Output_Get:Rotator)
   n9 = BreakTransform(InTransform=GetSocketTransform(self=Get Body_Mesh(), InSocketName=deform_upperarm_fr_r, TransformSpace=RTS_ParentBoneSpace))
   > Set deform_upperarm_fr_r_Translation(deform_upperarm_fr_r_Translation:Vector=@n9) -> (Output_Get:Vector)
   > Set deform_upperarm_fr_r_Rotation(deform_upperarm_fr_r_Rotation:Rotator=@n9.Rotation) -> (Output_Get:Rotator)
   n10 = BreakTransform(InTransform=GetSocketTransform(self=Get Body_Mesh(), InSocketName=upperarm_twist_01_r, TransformSpace=RTS_ParentBoneSpace))
   > Set upperarm_twist_01_r_Translation(upperarm_twist_01_r_Translation:Vector=@n10) -> (Output_Get:Vector)
   > Set upperarm_twist_01_r_Rotation(upperarm_twist_01_r_Rotation:Rotator=@n10.Rotation) -> (Output_Get:Rotator)
   n11 = Get Skel_Mesh()
   Set avoidClippBow(
    execute=@n1.then_1
    avoidClippBow:double=GetAlphaFromReferenceTransformDelta(SkeletalMeshComponent=@n11, BoneName=head, BaseName=neck_01, InputRange=(X=-5,Y=-20), TwistAxis="0,0,1")
   ) -> (Output_Get:double)
   > Set avoidClippBow_Extreme(avoidClippBow_Extreme:double=GetAlphaFromReferenceTransformDelta(SkeletalMeshComponent=@n11, BoneName=head, BaseName=neck_01, InputRange=(X=-30,Y=-55), TwistAxis="0,0,1")) -> (Output_Get:double)
   n12 = Get bIsSprinting()
   > Set sprintFloat(sprintFloat:double="To Float (Boolean)"(InBool=NOT Boolean(A=@n12))) -> (Output_Get:double)
   Event OnCharacterCustomizationLoaded()
   > n13 = Cast To PlayerPawn_Athena(Object=TryGetPawnOwner())
   > Set Body_Mesh(Body_Mesh=GetSkeletalMeshForPartType(self=@n13))
   Event BlueprintInitializeAnimation()
   > Set Skel_Mesh(Skel_Mesh=GetOwnerSkeletalMeshComponent())
  ]}
]
variables [
 avoidClippBow {type=number}
 avoidClippBow_Extreme {type=number}
 Body_Mesh {type=SkeletalMeshComponent}
 deform_pec_base_l_Rotation {type=Rotator}
 deform_pec_base_l_Translation {type=Vector}
 deform_pec_base_r_Rotation {type=Rotator}
 deform_pec_base_r_Translation {type=Vector}
 deform_pec_l_Rotation {type=Rotator}
 deform_pec_l_Translation {type=Vector}
 deform_pec_r_Rotation {type=Rotator}
 deform_pec_r_Translation {type=Vector}
 deform_upper_up_l_Rotation {type=Rotator}
 deform_upper_up_l_Translation {type=Vector}
 deform_upper_up_r_Rotation {type=Rotator}
 deform_upper_up_r_Translation {type=Vector}
 deform_upperarm_fr_l_Rotation {type=Rotator}
 deform_upperarm_fr_l_Translation {type=Vector}
 deform_upperarm_fr_r_Rotation {type=Rotator}
 deform_upperarm_fr_r_Translation {type=Vector}
 Skel_Mesh {type=SkeletalMeshComponent}
 sprintFloat {type=number}
 upperarm_twist_01_r_Rotation {type=Rotator}
 upperarm_twist_01_r_Translation {type=Vector}
]
